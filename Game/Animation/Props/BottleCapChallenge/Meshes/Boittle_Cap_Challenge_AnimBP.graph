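# Animation Blueprint for the Bottle Cap Challenge gadget prop that adds procedural physics sway to its tassle.
# The input pose passes through two Anim Dynamics nodes in series, one on the tassle bone and the next on tassleMid, before reaching the output.
# Both simulate in world space with angular limits and scaled, clamped component motion.
# The event graph holds only unconnected default nodes.
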AnimBlueprint /Game/Animation/Props/BottleCapChallenge/Meshes/Boittle_Cap_Challenge_AnimBP
target_skeleton=&/Bottle_Cap_Challenge_Gadget_Skeleton
graphs [
 AnimGraph {nodes [
   n1 = Anim Dynamics - Bone: tassle(ComponentPose=Local To Component(LocalPose=Input Pose()), Alpha=1, ExternalForce="0,0,0") {
    bGravityOverrideInSimSpace=true
    simulationSpace=World
    alphaBoolBlend {blendOption=Linear}
    boundBone {boneName=tassle}
    chainEnd {boneName=tassleMid}
    componentAppliedLinearAccClamp=<2000 2000 2000>
    componentLinearAccScale=<0.5 0.5 0.5>
    componentLinearVelScale=<0.1 0.1 0.1>
    physicsBodyDefinitions [{
      sphereCollisionRadius=0
      boundBone {boneName=tassle}
      boxExtents=<10 10 10>
      constraintSetup {angularLimitsMax=<0 8 4>, angularLimitsMin=<0 -8 -15>, angularTarget=<1 0 0>}
      localJointOffset=<1 0 0>
     }]
    relativeSpaceBone {boneName=root}
   }
   n2 = Anim Dynamics - Bone: tassleMid(ComponentPose=@n1, Alpha=1, ExternalForce="0,0,0") {
    bGravityOverrideInSimSpace=true
    simulationSpace=World
    alphaBoolBlend {blendOption=Linear}
    boundBone {boneName=tassleMid}
    chainEnd {boneName=tassleMid}
    componentAppliedLinearAccClamp=<1000 1000 1000>
    componentLinearAccScale=<0.5 0.5 0.5>
    componentLinearVelScale=<0.1 0.1 0.1>
    physicsBodyDefinitions [{
      sphereCollisionRadius=0
      boundBone {boneName=tassleMid}
      boxExtents=<10 5 10>
      constraintSetup {angularLimitsMax=<0 10 3>, angularLimitsMin=<0 -10 -20>, angularTarget=<1 0 0>}
      localJointOffset=<2 0 0>
     }]
   }
   Output Pose(Result=Component To Local(ComponentPose=@n2)) {layerGroup=None}
  ]}
 EventGraph {nodes [
   TryGetPawnOwner()
   Event BlueprintUpdateAnimation()
  ]}
]
variables []
